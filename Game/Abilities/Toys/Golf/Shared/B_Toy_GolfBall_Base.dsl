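; Base golf ball toy, a physics ball that tracks strokes and its starting hole for the golf minigame.
; On initialization, authority takes the hole number from an overlapped tee box, and later impulses increment strokes if HitOtherPawn is set.
; When stopping on stroke one, it cues a distance marker and notifies the matching hole if on its green POI.
; Scoring plays a hole-in-one, multi-stroke or wrong-hole cue, then silently destroys the ball.
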(event EventProjectileStop
  (params (ImpactResult :type Hit%20Result%20Structure%20%28by%20ref%29))
  (|CallParentFunction ImpactResult)
  ;[0 0 0] Show the distance marker when the ball comes to rest after the first stroke
  (Networking|SwitchHasAuthority
    (:Authority
      (if (== (Variables|Default|GetStrokeCount) 1)
        (CallFunction|PlayCueCenteredonMe (Variables|Default|GetCue_ShowDistanceMarker) (Math|Vector|VectorLengthXY (- (Variables|Default|GetInitialSpawnPosition) (Transformation|GetActorLocation))))
        ;[0 0 0] Let the hole know if we landed on the green
        (bind (_ongreen _holeinstance) (CallFunction|IsonGreen))
        (if _ongreen
          (Class|BGolfMinigameHole|LandedonGreen _holeinstance self))))))

(event Visuals|EventInitializeToyInstance
  (params (OwningPC :type Fort%20Player%20Controller%20Object%20Reference) (NumTimesSummoned :type Integer))
  (|CallParentFunction OwningPC NumTimesSummoned)
  (Variables|Default|SetStrokeCount 1)
  (Networking|SwitchHasAuthority
    (:Authority
      (CallFunction|FindandSetHoleNumber))))

(event EventApplyImpulseToBall
  (params (HitLocation :type Vector :default "0, 0, 0") (HitVelocity :type Vector :default "0, 0, 0") (HitMagnitude :type Float%20%28double-precision%29))
  (|CallParentFunction HitLocation HitVelocity HitMagnitude)
  (if (Variables|Default|GetHitOtherPawn)
    (Math|Integer|IncrementInt (Variables|Default|GetStrokeCount))))

(event Visuals|EventInitializeToyPreviewInFrontend
  (|CallParentFunction)
  (Components|Activation|SetActive (Variables|Default|GetGolfBallTrail) false false))

(fn FindAndSetHoleNumber
  (for _array_element (Collision|GetOverlappingActors (Variables|FortPhysicsBall|GetSphereCollision) "/Game/Abilities/Toys/Golf/Minigame/B_GolfMinigame_TeeBox.B_GolfMinigame_TeeBox_C" :!class /Script/Engine.PrimitiveComponent)
    (bind _asb_golf_minigame_tee_box (Utilities|Casting|CastToB_GolfMinigame_TeeBox _array_element))
    (bind _output_get (Variables|Default|SetStartedAtTeeBoxNumber (Variables|/Game/Abilities/Toys/Golf/Minigame/B_GolfMinigame_TeeBox.B_GolfMinigame_TeeBox_C|GetHoleNumber _asb_golf_minigame_tee_box)))
    (Development|PrintString (Utilities|String|ToString(Integer) _output_get) false true "(R=1.000000,G=0.000000,B=0.972075,A=1.000000)" 5.0)))

(fn FindGreensPOI
  (bind _outactors (Actor|GetAllActorsOfClass "/Game/Abilities/Toys/Golf/Minigame/B_GolfMinigame_Hole.B_GolfMinigame_Hole_C"))
  (for _array_element _outactors
    (if (== (Variables|/Game/Abilities/Toys/Golf/Minigame/B_GolfMinigame_Hole.B_GolfMinigame_Hole_C|GetHoleNumber _array_element) (Variables|Default|GetStartedAtTeeBoxNumber))
      (return (Variables|/Game/Abilities/Toys/Golf/Minigame/B_GolfMinigame_Hole.B_GolfMinigame_Hole_C|GetGreensPOI _array_element) _array_element)))
  (return 0 0))

(fn GetStrokeCountForHole
  (params (HoleNumber :type Integer))
  (return (Variables|Default|GetStrokeCount) (and (== HoleNumber (Variables|Default|GetStartedAtTeeBoxNumber)) (> (Variables|Default|GetStartedAtTeeBoxNumber) 0))))

(fn IsOnGreen
  (bind (_greenspoi _holeinstance) (CallFunction|FindGreensPOI))
  (Utilities|IsValid _greenspoi
    (:Is%20Valid
      (bind _returnvalue (Class|FortPoiVolume|DoesShapeComponentOverlap _greenspoi (Variables|FortPhysicsBall|GetSphereCollision)))
      (return _returnvalue _holeinstance))
    (:Is%20Not%20Valid
      (return false 0))))

(fn PlayGolfScoringCueAndDestruct
  (params (HoleNumber :type Integer))
  (Utilities|FlowControl|Sequence
    (:then_0
      (bind (_numstrokes _iscorrecthole) (CallFunction|GetStrokeCountforHole HoleNumber))
      (if _iscorrecthole
        (LocalVariables|SetCueToPlay (select (== _numstrokes 1) (Variables|Default|GetCue_HoleInOne) (Variables|Default|GetCue_ScoreGolfBall_MultipleStrokes)))
        (else
          (LocalVariables|SetCueToPlay (Variables|Default|GetCue_ScoreGolfBall_WrongOrNoHole)))))
    (:then_1
      (CallFunction|PlayCueCenteredonMe (LocalVariables|GetCueToPlay))
      (Gameplay|SilentDie false))))

(fn ConstructionScript
  (|CallParentFunction))
